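# Physics asset for the M_MED_Rooster male medium body cosmetic, covering a mirrored left and right pair of dynamic plate bones.
# Each plate has a large box body constrained to a tiny capsule body on its parent bone, which serves as its anchor.
# The constraints set modest, asymmetric swing limits and a near-zero twist limit, suggesting hinged secondary sway with almost no roll.
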PhysicsAsset /BRCosmetics/Characters/Player/Male/Medium/Bodies/M_MED_Rooster/Meshes/M_MED_Rooster_Physics
bodies {
 dyn_plate_l [{shapeType=Box, extentX=19.675, extentY=10.2525, extentZ=58.5214, center=<0.1 0 0>, rotation=<rot 0 -94.1393 -83.9246>}]
 dyn_plate_r [{shapeType=Box, extentX=19.675, extentY=10.2535, extentZ=58.5213, center=<-0.1 0 0>, rotation=<rot 0 -94.1392 -83.9197>}]
 parent_dyn_plate_l [{shapeType=Capsule, length=0.505, radius=0.505, rotation=<rot 0 -90 0>}]
 parent_dyn_plate_r [{shapeType=Capsule, length=0.505, radius=0.505, rotation=<rot 0 -90 0>}]
}
constraints [
 {bone1Name=dyn_plate_l, bone2Name=parent_dyn_plate_l, swing1LimitDegrees=10, swing2LimitDegrees=20, twistLimitDegrees=0.05}
 {bone1Name=dyn_plate_r, bone2Name=parent_dyn_plate_r, swing1LimitDegrees=10, swing2LimitDegrees=20, twistLimitDegrees=0.05}
]
