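# Post-process style cosmetic AnimBP for the male medium Rooster body, copying its pose from the attached parent mesh.
# Alpha-driven additive bone offsets adjust the rear collar and left plate, then spring-based AnimDynamics sway both plates, weighted by CrouchingRigidBodyAlpha.
# Rotation constraints make both hands' ring fingers follow the middle fingers.
# A RigidBody node is present but unconnected to the output, and the EventGraph is empty.
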AnimBlueprint /BRCosmetics/Characters/Player/Male/Medium/Bodies/M_MED_Rooster/Meshes/M_MED_Rooster_AnimBP
target_skeleton=&/M_MED_Rooster_Skeleton
graphs [
 AnimGraph {nodes [
   n1 = Get CrouchingRigidBodyAlpha() -> (CrouchingRigidBodyAlpha:float)
   RigidBody(Alpha=@n1) {
    bClampLinearTranslationLimitToRefPose=true
    bOverrideWorldGravity=true
    overridePhysicsAsset=&/M_MED_Rooster_Physics
    alphaBoolBlend {blendOption=Linear}
    baseBoneRef {boneName=spine_01}
    componentLinearAccScale=<0.5 0.5 0.5>
    componentLinearVelScale=<0.5 0.5 0.5>
    overrideWorldGravity {z=-980}
   }
   n2 = "Transform (Modify) Bone - Bone: dyn_collar_rear"(
    ComponentPose=Local To Component(LocalPose=Copy Pose From Mesh() {bUseAttachedParent=true})
    Alpha=Get HeadRotationBackAlpha() -> (HeadRotationBackAlpha:float)
    Translation="0,0,0"
    Rotation="0.000,-15,0"
    Scale="1,1,1"
   ) {
    rotationMode=BMM_Additive
    rotationSpace=BCS_BoneSpace
    alphaBoolBlend {blendOption=Linear}
    boneToModify {boneName=dyn_collar_rear}
    rotation {yaw=-15}
   }
   n3 = "Transform (Modify) Bone - Bone: parent_dyn_plate_l"(ComponentPose=@n2, Alpha=Get PouchAlpha() -> (PouchAlpha:float), Translation="-0.000,-4,0.000", Rotation="3.421,-1.343,-30", Scale="1,1,1") {
    rotationMode=BMM_Additive
    rotationSpace=BCS_BoneSpace
    translationMode=BMM_Additive
    translationSpace=BCS_BoneSpace
    alphaBoolBlend {blendOption=Linear}
    boneToModify {boneName=parent_dyn_plate_l}
    rotation=<rot 3.421 -1.343 -30>
    translation {y=-4}
   }
   n4 = Anim Dynamics - Bone: dyn_plate_l(ComponentPose=@n3, Alpha=@n1, ExternalForce="0,0,0") {
    angularDampingOverride=0.9999
    angularSpringConstant=10000
    bAngularSpring=true
    bGravityOverrideInSimSpace=true
    bOverrideAngularDamping=true
    gravityScale=0.33
    simulationSpace=RootRelative
    alphaBoolBlend {blendOption=Linear}
    boundBone {boneName=dyn_plate_l}
    componentAppliedLinearAccClamp=<1500 1500 1200>
    componentLinearAccScale=<0.5 0.5 0.7>
    componentLinearVelScale=<0.5 0.5 0.7>
    physicsBodyDefinitions [{
      sphereCollisionRadius=0
      boundBone {boneName=dyn_plate_l}
      boxExtents=<7 40 40>
      constraintSetup {angularLimitsMax=<0 3 25>, angularLimitsMin=<0 -15 -25>, angularTarget=<0 0 0>}
      localJointOffset=<6 0 0>
     }]
   }
   n5 = Anim Dynamics - Bone: dyn_plate_r(ComponentPose=@n4, Alpha=@n1, ExternalForce="0,0,0") {
    angularDampingOverride=0.9999
    angularSpringConstant=10000
    bAngularSpring=true
    bGravityOverrideInSimSpace=true
    bOverrideAngularDamping=true
    gravityScale=0.33
    simulationSpace=RootRelative
    alphaBoolBlend {blendOption=Linear}
    boundBone {boneName=dyn_plate_r}
    componentAppliedLinearAccClamp=<1500 1500 1200>
    componentLinearAccScale=<0.5 0.5 0.7>
    componentLinearVelScale=<0.5 0.5 0.7>
    physicsBodyDefinitions [{
      sphereCollisionRadius=0
      boundBone {boneName=dyn_plate_r}
      boxExtents=<7 40 40>
      constraintSetup {angularLimitsMax=<0 3 25>, angularLimitsMin=<0 -15 -25>, angularTarget=<0 0 0>}
      localJointOffset=<-6 0 0>
     }]
   }
   n6 = "Constraint - ring_metacarpal_l (R)"(ComponentPose=@n5, Alpha=1, ConstraintWeights_0=1) {
    alphaBoolBlend {blendOption=Linear}
    boneToModify {boneName=ring_metacarpal_l}
    constraintSetup [{transformType=Rotation, targetBone {boneName=middle_metacarpal_l}}]
    constraintWeights [1]
   }
   n7 = "Constraint - ring_01_l (R)"(ComponentPose=@n6, Alpha=1, ConstraintWeights_0=1) {
    alphaBoolBlend {blendOption=Linear}
    boneToModify {boneName=ring_01_l}
    constraintSetup [{transformType=Rotation, targetBone {boneName=middle_01_l}}]
    constraintWeights [1]
   }
   n8 = "Constraint - ring_02_l (R)"(ComponentPose=@n7, Alpha=1, ConstraintWeights_0=1) {
    alphaBoolBlend {blendOption=Linear}
    boneToModify {boneName=ring_02_l}
    constraintSetup [{transformType=Rotation, targetBone {boneName=middle_02_l}}]
    constraintWeights [1]
   }
   n9 = "Constraint - ring_03_l (R)"(ComponentPose=@n8, Alpha=1, ConstraintWeights_0=1) {
    alphaBoolBlend {blendOption=Linear}
    boneToModify {boneName=ring_03_l}
    constraintSetup [{transformType=Rotation, targetBone {boneName=middle_03_l}}]
    constraintWeights [1]
   }
   n10 = "Constraint - ring_metacarpal_r (R)"(ComponentPose=@n9, Alpha=1, ConstraintWeights_0=1) {
    alphaBoolBlend {blendOption=Linear}
    boneToModify {boneName=ring_metacarpal_r}
    constraintSetup [{transformType=Rotation, targetBone {boneName=middle_metacarpal_r}}]
    constraintWeights [1]
   }
   n11 = "Constraint - ring_01_r (R)"(ComponentPose=@n10, Alpha=1, ConstraintWeights_0=1) {
    alphaBoolBlend {blendOption=Linear}
    boneToModify {boneName=ring_01_r}
    constraintSetup [{transformType=Rotation, targetBone {boneName=middle_01_r}}]
    constraintWeights [1]
   }
   n12 = "Constraint - ring_02_r (R)"(ComponentPose=@n11, Alpha=1, ConstraintWeights_0=1) {
    alphaBoolBlend {blendOption=Linear}
    boneToModify {boneName=ring_02_r}
    constraintSetup [{transformType=Rotation, targetBone {boneName=middle_02_r}}]
    constraintWeights [1]
   }
   n13 = "Constraint - ring_03_r (R)"(ComponentPose=@n12, Alpha=1, ConstraintWeights_0=1) {
    alphaBoolBlend {blendOption=Linear}
    boneToModify {boneName=ring_03_r}
    constraintSetup [{transformType=Rotation, targetBone {boneName=middle_03_r}}]
    constraintWeights [1]
   }
   Output Pose(Result=Component To Local(ComponentPose=@n13)) {layerGroup=None}
  ]}
 EventGraph {links [], nodes []}
]
variables []
